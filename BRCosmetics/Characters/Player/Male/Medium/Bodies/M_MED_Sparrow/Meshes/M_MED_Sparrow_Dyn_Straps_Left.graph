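# Physics asset for the medium male Sparrow body that sets up dangling left-leg straps and chains as jointed capsule bodies.
# A four-link strap and a two-link chain each hang from a small master body, joined to the left thigh twist bone with free swing and twist.
# Angular limits restrict each link, tightest at the strap root.
# Large capsules on the left thigh and calf give the leg collision volume.
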PhysicsAsset /BRCosmetics/Characters/Player/Male/Medium/Bodies/M_MED_Sparrow/Meshes/M_MED_Sparrow_Dyn_Straps_Left
bodies {
 calf_l [{
   shapeType=Capsule
   length=27.5971
   radius=14.5753
   shapeName=thigh_twist_01_r_capsule
   center=<-20.4149 -5.521 7.7318>
   rotation=<rot 88.5415 0.0012 -178.4839>
  }]
 dyn_chains_01 [{shapeType=Capsule, length=1.8424, radius=4.1576, shapeName=dyn_straps_r_01_capsule, center=<4 0 1.5>, rotation=<rot 0 0 90>}]
 dyn_chains_02 [{shapeType=Capsule, length=2.3378, radius=5.1622, shapeName=dyn_straps_l_02_capsule, center=<3 0 1.5>, rotation=<rot 0 0 90>}]
 dyn_chains_MSTR [{shapeType=Capsule, length=0.505, radius=0.505, shapeName=dyn_straps_l_MSTR_capsule, rotation=<rot 0 -90 0>}]
 dyn_straps_l_01 [{shapeType=Capsule, length=3.7143, radius=3.0468, shapeName=dyn_straps_r_01_capsule, center=<2 0 1.5>, rotation=<rot 90 0 0>}]
 dyn_straps_l_02 [{shapeType=Capsule, length=8.3147, radius=3.0468, shapeName=dyn_straps_l_02_capsule, center=<5 0 1.5>, rotation=<rot 90 0 0>}]
 dyn_straps_l_03 [{shapeType=Capsule, length=8.5647, radius=2.7968, shapeName=dyn_straps_l_03_capsule, center=<5 0 1.5>, rotation=<rot 90 0 0>}]
 dyn_straps_l_04 [{shapeType=Capsule, length=8.8147, radius=2.5468, shapeName=dyn_straps_l_04_capsule, center=<5 0 1.5>, rotation=<rot 90 0 0>}]
 dyn_straps_l_MSTR [{shapeType=Capsule, length=0.505, radius=0.505, shapeName=dyn_straps_l_MSTR_capsule, rotation=<rot 0 -90 0>}]
 thigh_twist_01_l [{shapeType=Capsule, length=39.1255, radius=6.6825, shapeName=thigh_twist_01_r_capsule, center=<3.6832 0 0>, rotation=<rot 90 0 180>}]
}
constraints [
 {bone1Name=dyn_straps_l_01, bone2Name=dyn_straps_l_MSTR, swing1LimitDegrees=10, swing2LimitDegrees=5, twistLimitDegrees=2}
 {bone1Name=dyn_straps_l_02, bone2Name=dyn_straps_l_01, swing1LimitDegrees=30, swing2LimitDegrees=45, twistLimitDegrees=5}
 {bone1Name=dyn_straps_l_03, bone2Name=dyn_straps_l_02, swing1LimitDegrees=30, swing2LimitDegrees=45, twistLimitDegrees=5}
 {bone1Name=dyn_straps_l_04, bone2Name=dyn_straps_l_03, swing1LimitDegrees=30, swing2LimitDegrees=45, twistLimitDegrees=5}
 {
  bone1Name=dyn_straps_l_MSTR
  bone2Name=thigh_twist_01_l
  swing1LimitDegrees=45
  swing1Motion=Free
  swing2LimitDegrees=45
  swing2Motion=Free
  twistLimitDegrees=45
  twistMotion=Free
 }
 {
  bone1Name=dyn_chains_MSTR
  bone2Name=thigh_twist_01_l
  swing1LimitDegrees=45
  swing1Motion=Free
  swing2LimitDegrees=45
  swing2Motion=Free
  twistLimitDegrees=45
  twistMotion=Free
 }
 {bone1Name=dyn_chains_01, bone2Name=dyn_chains_MSTR, swing1LimitDegrees=45, swing2LimitDegrees=45, twistLimitDegrees=5}
 {bone1Name=dyn_chains_02, bone2Name=dyn_chains_01, swing1LimitDegrees=45, swing2LimitDegrees=45, twistLimitDegrees=5}
]
