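# Physics asset for the right-side dangling straps on the medium male Sparrow body cosmetic, simulating their secondary motion.
# A master bone and four capsule bodies form a chain hung from the right thigh twist bone, with thigh and calf capsules as leg volumes.
# The master-to-thigh constraint has free rotation; chain links get angular limits with near-zero twist, and the first link is tightest.
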PhysicsAsset /BRCosmetics/Characters/Player/Male/Medium/Bodies/M_MED_Sparrow/Meshes/M_MED_Sparrow_Dyn_Straps_Right
bodies {
 calf_r [{shapeType=Capsule, length=28.8788, radius=6.5379, shapeName=thigh_twist_01_r_capsule, center=<19.9323 1.5782 0>, rotation=<rot 90 0 0>}]
 dyn_straps_r_01 [{shapeType=Capsule, length=3.4643, radius=3.2968, shapeName=dyn_straps_r_01_capsule, center=<2 0 1.5>, rotation=<rot 90 0 0>}]
 dyn_straps_r_02 [{shapeType=Capsule, length=6.9856, radius=3.0468, shapeName=dyn_straps_r_02_capsule, center=<5 0 1.5>, rotation=<rot 90 0 0>}]
 dyn_straps_r_03 [{shapeType=Capsule, length=7.4856, radius=2.5468, shapeName=dyn_straps_r_03_capsule, center=<5 0 1.5>, rotation=<rot 90 0 0>}]
 dyn_straps_r_04 [{shapeType=Capsule, length=10.2356, radius=2.5468, shapeName=dyn_straps_r_04_capsule, center=<6.6937 0 1.5>, rotation=<rot 90 0 0>}]
 dyn_straps_r_MSTR [{shapeType=Capsule, length=0.505, radius=0.505, shapeName=dyn_straps_r_MSTR_capsule, rotation=<rot 0 -90 0>}]
 thigh_twist_01_r [{shapeType=Capsule, length=26.1255, radius=6.6825, shapeName=thigh_twist_01_r_capsule, center=<3.387 0 0>, rotation=<rot 90 0 0>}]
}
constraints [
 {
  bone1Name=dyn_straps_r_MSTR
  bone2Name=thigh_twist_01_r
  swing1LimitDegrees=45
  swing1Motion=Free
  swing2LimitDegrees=45
  swing2Motion=Free
  twistLimitDegrees=45
  twistMotion=Free
 }
 {bone1Name=dyn_straps_r_01, bone2Name=dyn_straps_r_MSTR, swing1LimitDegrees=20, swing2LimitDegrees=2, twistLimitDegrees=2}
 {bone1Name=dyn_straps_r_02, bone2Name=dyn_straps_r_01, swing1LimitDegrees=45, swing2LimitDegrees=45, twistLimitDegrees=2}
 {bone1Name=dyn_straps_r_03, bone2Name=dyn_straps_r_02, swing1LimitDegrees=45, swing2LimitDegrees=45, twistLimitDegrees=2}
 {bone1Name=dyn_straps_r_04, bone2Name=dyn_straps_r_03, swing1LimitDegrees=45, swing2LimitDegrees=45, twistLimitDegrees=2}
]
