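# Physics asset for the Sparrow male medium body's dangling necklace, giving its bone chain capsule bodies for secondary motion.
# Capsules on the upper spine and both clavicles form the torso shapes around the necklace.
# A master bone joins spine_05 with free rotation and leads a three-segment chain.
# Chain joints allow a wide second swing but little twist, with first swing tightening toward the tip.
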PhysicsAsset /BRCosmetics/Characters/Player/Male/Medium/Bodies/M_MED_Sparrow/Meshes/M_MED_Sparrow_dyn_necklace
bodies {
 clavicle_l [{
   shapeType=Capsule
   length=8.5904
   radius=7.0039
   shapeName=clavicle_l_capsule
   center=<11.3085 -0.2 -3.4318>
   rotation=<rot 90 19.7672 17.8373>
  }]
 clavicle_r [{shapeType=Capsule, length=8.5904, radius=7.0039, shapeName=clavicle_l_capsule, center=<-11.3085 0.2 3.4318>, rotation=<rot 90 -90 88.0701>}]
 dyn_necklace_01 [{shapeType=Capsule, length=3, radius=2.7372, center=<3 0 2.5>, rotation=<rot 84.2618 -179.9999 -179.9999>}]
 dyn_necklace_02 [{shapeType=Capsule, length=3, radius=3.2592, center=<3 0 3>, rotation=<rot 85.1983 0 0>}]
 dyn_necklace_03 [{shapeType=Capsule, length=4.75, radius=3.1002, center=<5 0 2.8063>, rotation=<rot 0 0 90>}]
 dyn_necklace_MSTR [{shapeType=Capsule, length=0.5, radius=0.5}]
 spine_04 [{shapeType=Capsule, length=15.2009, radius=12.4611, center=<4.4101 -0.3674 0.0001>, rotation=<rot 0 89.9999 0>}]
 spine_05 [{shapeType=Capsule, length=19.0759, radius=11.9611, center=<-2.1676 -1.7543 0.0001>, rotation=<rot 0 89.9999 0>}]
}
constraints [
 {
  bone1Name=dyn_necklace_MSTR
  bone2Name=spine_05
  swing1LimitDegrees=45
  swing1Motion=Free
  swing2LimitDegrees=45
  swing2Motion=Free
  twistLimitDegrees=45
  twistMotion=Free
 }
 {bone1Name=dyn_necklace_01, bone2Name=dyn_necklace_MSTR, swing1LimitDegrees=8, swing2LimitDegrees=30, twistLimitDegrees=2}
 {bone1Name=dyn_necklace_02, bone2Name=dyn_necklace_01, swing1LimitDegrees=4, swing2LimitDegrees=30, twistLimitDegrees=2}
 {
  bone1Name=spine_05
  bone2Name=spine_04
  swing1LimitDegrees=45
  swing1Motion=Free
  swing2LimitDegrees=45
  swing2Motion=Free
  twistLimitDegrees=45
  twistMotion=Free
 }
 {
  bone1Name=clavicle_l
  bone2Name=spine_05
  swing1LimitDegrees=45
  swing1Motion=Free
  swing2LimitDegrees=45
  swing2Motion=Free
  twistLimitDegrees=45
  twistMotion=Free
 }
 {
  bone1Name=clavicle_r
  bone2Name=spine_05
  swing1LimitDegrees=45
  swing1Motion=Free
  swing2LimitDegrees=45
  swing2Motion=Free
  twistLimitDegrees=45
  twistMotion=Free
 }
 {bone1Name=dyn_necklace_03, bone2Name=dyn_necklace_02, swing1LimitDegrees=2, swing2LimitDegrees=30, twistLimitDegrees=2}
]
