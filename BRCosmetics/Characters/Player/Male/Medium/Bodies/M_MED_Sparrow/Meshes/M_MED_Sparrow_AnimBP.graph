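# Post-process style AnimBP for the Sparrow medium male body, driving its cape and backpack attach bones, necklace, straps and chains.
# Copies the parent mesh pose through a deform Control Rig, constraints and necklace simulation (AnimDynamics, RigidBody) into a cached pose.
# Leg straps then use per-side RigidBody simulation, with a chain look-at fallback when rigid bodies are disabled.
# Event graph derives corrective alphas from bone twists and skydive-aware gravity via a deprecated function.
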AnimBlueprint /BRCosmetics/Characters/Player/Male/Medium/Bodies/M_MED_Sparrow/Meshes/M_MED_Sparrow_AnimBP
target_skeleton=&/M_MED_Sparrow_Skeleton
graphs [
 AnimGraph {nodes [
   n1 = "Transform (Modify) Bone - Bone: dyn_chains_MSTR"(
    ComponentPose=Local To Component(LocalPose=Use cached pose 'basePose'())
    Alpha=Get thigh_rot_up_l() -> (thigh_rot_up_l:double)
    Translation="-2.147,-4.090,-1.239"
    Rotation="0,-80,-20"
    Scale="1,1,1"
   ) {
    rotationMode=BMM_Additive
    rotationSpace=BCS_BoneSpace
    translationMode=BMM_Additive
    translationSpace=BCS_BoneSpace
    alphaBoolBlend {blendOption=Linear}
    boneToModify {boneName=dyn_chains_MSTR}
    rotation {roll=-20, yaw=-80}
    translation=<-2.147 -4.09 -1.239>
   }
   n2 = "Transform (Modify) Bone - Bone: dyn_straps_l_MSTR"!(ComponentPose=@n1, bAlphaBoolEnabled:bool=Get bIsJumping(), Translation="0,0,0", Rotation="25,0,0", Scale="1,1,1") -> (Pose) {
    alphaInputType=Bool
    rotationMode=BMM_Additive
    rotationSpace=BCS_BoneSpace
    translationMode=BMM_Additive
    translationSpace=BCS_BoneSpace
    alphaBoolBlend {blendInTime=0.3, blendOption=Linear, blendOutTime=0.3}
    boneToModify {boneName=dyn_straps_l_MSTR}
    rotation {pitch=25}
   }
   n3 = Get GravityOverride()
   n4 = RigidBody!(ComponentPose=@n2, Alpha, OverrideWorldGravity:Vector=@n3, ExternalForce) -> (Pose) {
    bClampLinearTranslationLimitToRefPose=true
    bOverrideWorldGravity=true
    overridePhysicsAsset=&/M_MED_Sparrow_Dyn_Straps_Left
    simulationSpace=BaseBoneSpace
    alphaBoolBlend {blendOption=Linear}
    baseBoneRef {boneName=thigh_l}
    componentAppliedLinearAccClamp=<3000 3000 3000>
    componentLinearAccScale {y=0.3, z=0.2}
    componentLinearVelScale {y=0.2, z=0.2}
    overrideWorldGravity {z=-980}
    simSpaceSettings {maxAngularAcceleration=100, maxAngularVelocity=5, maxLinearAcceleration=300, maxLinearVelocity=50, velocityScaleZ=0, worldAlpha=1}
   }
   n5 = RigidBody(ComponentPose=@n4, OverrideWorldGravity=@n3) {
    bClampLinearTranslationLimitToRefPose=true
    bOverrideWorldGravity=true
    overridePhysicsAsset=&/M_MED_Sparrow_Dyn_Straps_Right
    simulationSpace=BaseBoneSpace
    alphaBoolBlend {blendOption=Linear}
    baseBoneRef {boneName=thigh_r}
    componentAppliedLinearAccClamp=<3000 3000 3000>
    componentLinearAccScale {y=0.3, z=0.2}
    componentLinearVelScale {y=0.2, z=0.2}
    overrideWorldGravity {z=-980}
    simSpaceSettings {maxAngularAcceleration=100, maxAngularVelocity=5, maxLinearAcceleration=300, maxLinearVelocity=50, velocityScaleZ=0, worldAlpha=1}
   }
   n6 = Look At - Bone: dyn_chains_MSTR(ComponentPose=Local To Component(LocalPose=Use cached pose 'basePose'()), Alpha=1) {
    alphaBoolBlend {blendOption=Linear}
    boneToModify {boneName=dyn_chains_MSTR}
    lookAtLocation=<-13.8986 29.6096 -5.6738>
    lookAtTarget {boneReference {boneName=thigh_twist_01_l}}
    lookAt_Axis {axis=<1 0>}
   }
   Output Pose(Result=Blend Poses by bool(BlendPose_0=Component To Local(ComponentPose=@n5), BlendPose_1=Component To Local(ComponentPose=@n6), bActiveValue=Get bRigidBodyEnabled()) {blendType=Linear}) {layerGroup=None}
   n7 = Control Rig(Source=Copy Pose From Mesh() {bUseAttachedParent=true}) {
    alphaBoolBlend {blendOption=Linear}
    controlRigAssetReference {blueprintRigClass=&/M_MED_Sparrow_DeformRig_ControlRig.M_MED_Sparrow_DeformRig_ControlRig_C}
   }
   n8 = "Constraint - attach_cape (PPPPPP)"!(
    ComponentPose=Local To Component(LocalPose=@n7)
    Alpha=1
    ConstraintWeights_0=0.0426
    ConstraintWeights_1:float=0.0402
    ConstraintWeights_2:float=0.8297
    ConstraintWeights_3:float=0.0476
    ConstraintWeights_4:float=0.0398
    ConstraintWeights_5:float=0
   ) -> (Pose) {
    alphaBoolBlend {blendOption=Linear}
    boneToModify {boneName=attach_cape}
    constraintSetup [
     {targetBone {boneName=spine_03}}
     {targetBone {boneName=spine_04}}
     {targetBone {boneName=spine_05}}
     {targetBone {boneName=clavicle_l}}
     {targetBone {boneName=clavicle_r}}
     {targetBone {boneName=upperarm_twist_01_r}}
    ]
    constraintWeights [1,1,1,1,1,1]
   }
   n9 = "Constraint - attach_backpack (PPPPPP)"!(
    ComponentPose=@n8
    Alpha=1
    ConstraintWeights_0=0.0183
    ConstraintWeights_1:float=0.2769
    ConstraintWeights_2:float=0.2277
    ConstraintWeights_3:float=0.4575
    ConstraintWeights_4:float=0.0137
    ConstraintWeights_5:float=0.006
   ) -> (Pose) {
    alphaBoolBlend {blendOption=Linear}
    boneToModify {boneName=attach_backpack}
    constraintSetup [
     {targetBone {boneName=spine_02}}
     {targetBone {boneName=spine_03}}
     {targetBone {boneName=spine_04}}
     {targetBone {boneName=spine_05}}
     {targetBone {boneName=clavicle_l}}
     {targetBone {boneName=clavicle_r}}
    ]
    constraintWeights [1,1,1,1,1,1]
   }
   n10 = "Constraint - dyn_straps_r_MSTR (PPPPP)"!(
    ComponentPose=@n9
    Alpha=1
    ConstraintWeights_0=0.0894
    ConstraintWeights_1:float=0.0586
    ConstraintWeights_2:float=0.8098
    ConstraintWeights_3:float=0.0137
    ConstraintWeights_4:float=0.0284
   ) -> (Pose) {
    alphaBoolBlend {blendOption=Linear}
    boneToModify {boneName=dyn_straps_r_MSTR}
    constraintSetup [
     {targetBone {boneName=pelvis}}
     {targetBone {boneName=thigh_r}}
     {targetBone {boneName=thigh_twist_01_r}}
     {targetBone {boneName=deform_glute_r}}
     {targetBone {boneName=deform_groin_r}}
    ]
    constraintWeights [1,1,1,1,1]
   }
   n11 = "Constraint - dyn_straps_l_MSTR (PPPPP)"!(
    ComponentPose=@n10
    Alpha=1
    ConstraintWeights_0=0.0608
    ConstraintWeights_1:float=0.0884
    ConstraintWeights_2:float=0.8093
    ConstraintWeights_3:float=0.0123
    ConstraintWeights_4:float=0.0292
   ) -> (Pose) {
    alphaBoolBlend {blendOption=Linear}
    boneToModify {boneName=dyn_straps_l_MSTR}
    constraintSetup [
     {targetBone {boneName=thigh_l}}
     {targetBone {boneName=pelvis}}
     {targetBone {boneName=thigh_twist_01_l}}
     {targetBone {boneName=deform_glute_l}}
     {targetBone {boneName=deform_groin_l}}
    ]
    constraintWeights [1,1,1,1,1]
   }
   n12 = "Constraint - dyn_chains_MSTR (PPPP)"!(
    ComponentPose=@n11
    Alpha=1
    ConstraintWeights_0=0.0187
    ConstraintWeights_1:float=0.5495
    ConstraintWeights_2:float=0.223
    ConstraintWeights_3:float=0.2088
   ) -> (Pose) {
    alphaBoolBlend {blendOption=Linear}
    boneToModify {boneName=dyn_chains_MSTR}
    constraintSetup [
     {targetBone {boneName=spine_01}}
     {targetBone {boneName=pelvis}}
     {targetBone {boneName=thigh_twist_01_l}}
     {targetBone {boneName=deform_groin_l}}
    ]
    constraintWeights [1,1,1,1]
   }
   n13 = Look At - Bone: dyn_necklace_MSTR(ComponentPose=@n12, Alpha=1) {
    alphaBoolBlend {blendOption=Linear}
    boneToModify {boneName=dyn_necklace_MSTR}
    lookAtLocation {x=5}
    lookAtTarget {boneReference {boneName=dyn_necklace_TGT}}
    lookAt_Axis {axis=<1 0>}
   }
   n14 = "Anim Dynamics - Chain: dyn_necklace_01 -> dyn_necklace_03"(ComponentPose=@n13, Alpha=1, ExternalForce="0,0,0") {
    angularDampingOverride=0.9999
    bChain=true
    bOverrideAngularDamping=true
    bUseGravityOverride=true
    simulationSpace=RootRelative
    alphaBoolBlend {blendOption=Linear}
    boundBone {boneName=dyn_necklace_01}
    chainEnd {boneName=dyn_necklace_03}
    componentAppliedLinearAccClamp=<500 500 500>
    componentLinearVelScale=<0.2 0.2 0.2>
    gravityOverride {z=-980}
    physicsBodyDefinitions [
     {
      boundBone {boneName=dyn_necklace_01}
      boxExtents=<10 15 15>
      constraintSetup {angularLimitsMax=<0 0 10>, angularLimitsMin=<0 -10 -10>, angularTarget=<0 0 0>}
      localJointOffset=<1.5 0 0>
     }
     {
      boundBone {boneName=dyn_necklace_02}
      boxExtents=<15 12 12>
      constraintSetup {angularLimitsMax=<0 0 10>, angularLimitsMin=<0 -15 -10>, angularTarget=<0 0 0>}
      localJointOffset=<1 0 0>
     }
     {
      boundBone {boneName=dyn_necklace_03}
      boxExtents=<15 10 10>
      constraintSetup {angularLimitsMax=<0 24 5>, angularLimitsMin=<0 -45 -5>, angularTarget=<0 0 0>}
      localJointOffset=<1 0 -0.5>
     }
    ]
    relativeSpaceBone {boneName=spine_04}
   }
   n15 = "Transform (Modify) Bone - Bone: dyn_necklace_02"(ComponentPose=@n14, Alpha=Get pullOutNecklace() -> (pullOutNecklace:double), Translation="0,0,1", Rotation="0,0,0", Scale="1,1,1") {
    translationMode=BMM_Additive
    translationSpace=BCS_BoneSpace
    alphaBoolBlend {blendOption=Linear}
    boneToModify {boneName=dyn_necklace_02}
    translation {z=1}
   }
   n16 = RigidBody(ComponentPose=@n15, OverrideWorldGravity="0,0,-980") {
    bClampLinearTranslationLimitToRefPose=true
    bOverrideWorldGravity=true
    overridePhysicsAsset=&/M_MED_Sparrow_dyn_necklace
    simulationSpace=BaseBoneSpace
    alphaBoolBlend {blendOption=Linear}
    baseBoneRef {boneName=spine_05}
    componentAppliedLinearAccClamp=<2000 2000 1000>
    componentLinearAccScale {x=0.4}
    componentLinearVelScale {x=0.4}
    overrideWorldGravity {z=-980}
    simSpaceSettings {maxAngularAcceleration=100, maxAngularVelocity=5, maxLinearAcceleration=300, maxLinearVelocity=50, velocityScaleZ=0, worldAlpha=1}
   }
   Save cached pose 'basePose'(Pose=Component To Local(ComponentPose=@n16)) {cacheName=basePose}
  ]}
 EventGraph {nodes [
   Event BlueprintUpdateAnimation()
   > n1 = Sequence()
   > Set pullOutNecklace(pullOutNecklace:double=GetAlphaFromReferenceTransformDelta(
     SkeletalMeshComponent=GetPartSkeletalMeshComponent()
     BoneName=dyn_necklace_01
     BaseName=dyn_necklace_MSTR
     DeltaComponentType=QuaternionTwist
     InputRange=(X=0,Y=-6)
     TwistAxis="0,0,1"
    )) -> (Output_Get:double)
   n2 = Get SkelMesh()
   > Set thigh_rot_up_l(thigh_rot_up_l:double=GetAlphaFromReferenceTransformDelta(SkeletalMeshComponent=@n2, BoneName=thigh_l, BaseName=pelvis, DeltaComponentType=QuaternionTwist, InputRange=(X=0,Y=100), TwistAxis="0,0,1")) -> (Output_Get:double)
   n3 = Gravity Override State Function DEPRECATED(
    execute=@n1.then_1
    CharacterPartAnimInstance=Self-Reference()
    GameTimeInSeconds=GetGameTimeInSeconds()
    GravityOverrideParams=Get Gravity Override Params()
   )
   > Set GravityOverride(GravityOverride=@n3)
   Event BlueprintInitializeAnimation()
   n4 = GetOwnerSkeletalMeshComponent()
   > Is Valid(InputObject=@n4)
   > Set SkelMesh(SkelMesh=@n4)
   GetAlphaFromReferenceTransformDelta(
    SkeletalMeshComponent=@n2
    BoneName=spine_04
    BaseName=spine_03
    DeltaComponentType=QuaternionTwist
    InputRange=(X=2,Y=-6)
    TwistAxis="1,0,0"
   )
  ]}
]
variables [
 Gravity Override Params {
  type=GravityOverrideParamsStruct
  default {
   skydive_DivingDown {windAmplitude=<50 800 100>, windFrequency=<1.5 30 10>, windOffset=<0 300 0>}
   skydive_FloatingDown {gravityOverride=<0 -700 0>, windAmplitude=<200 -500 200>, windOffset=<0 -600 100>}
   skydive_GoingUp {gravityOverride=<0 -300 200>, windAmplitude=<200 -500 300>}
   skydive_Parachuting {gravityOverride=<0 -300 -980>, windAmplitude=<200 -150 300>, windFrequency=<1.5 4 6>, windOffset=<0 -100 0>}
  }
 }
 GravityOverride {type=Vector}
 pullOutNecklace {type=number}
 SkelMesh {type=SkeletalMeshComponent}
 thigh_rot_up_l {type=number}
]
