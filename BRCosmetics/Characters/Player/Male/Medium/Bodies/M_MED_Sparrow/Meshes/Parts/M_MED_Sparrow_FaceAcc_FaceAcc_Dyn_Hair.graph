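# Physics asset for the dynamic back hair on the face accessory mesh of the medium male Sparrow player body.
# Center, left and right chains each run three capsules from a small master body linked to the head, with limited swing and twist, tightest twist at the root.
# Head, neck, upper spine, clavicle and upper-arm capsules supply upper-body shapes around the hair.
# Master, head, neck and upper-arm joints rotate freely, while clavicles are limited against the spine.
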PhysicsAsset /BRCosmetics/Characters/Player/Male/Medium/Bodies/M_MED_Sparrow/Meshes/Parts/M_MED_Sparrow_FaceAcc_FaceAcc_Dyn_Hair
bodies {
 clavicle_l [{
   shapeType=Capsule
   length=3.875
   radius=7.25
   shapeName=clavicle_l_capsule
   center=<8.8902 0.4581 -0.213>
   rotation=<rot 87.4584 -12.5403 -20.0159>
  }]
 clavicle_r [{
   shapeType=Capsule
   length=3.6654
   radius=7.4596
   shapeName=clavicle_l_capsule
   center=<-8.8902 -0.4581 0.213>
   rotation=<rot 87.4584 -12.5403 159.9841>
  }]
 dyn_hair_bk_01 [{shapeType=Capsule, length=3.75, radius=6.25, shapeName=dyn_hair_bk_01_capsule, center=<5 0 4>, rotation=<rot 90 0 0>}]
 dyn_hair_bk_02 [{shapeType=Capsule, length=4.4643, radius=6.2857, shapeName=dyn_hair_bk_01_capsule, center=<5 0 2>, rotation=<rot 0 0 90>}]
 dyn_hair_bk_03 [{shapeType=Capsule, length=5.6199, radius=8.75, shapeName=dyn_hair_bk_01_capsule, center=<7 0 7>, rotation=<rot 0 0 90>}]
 dyn_hair_bk_MSTR [{shapeType=Capsule, length=0.505, radius=0.505, shapeName=dyn_hair_bk_MSTR_capsule, rotation=<rot 0 -90 0>}]
 dyn_hair_bk_l_01 [{shapeType=Capsule, length=4, radius=6, shapeName=dyn_hair_bk_01_capsule, center=<5 0 4>, rotation=<rot 84.9999 -0.0001 -0.0002>}]
 dyn_hair_bk_l_02 [{shapeType=Capsule, length=2.1511, radius=5.9477, shapeName=dyn_hair_bk_01_capsule, center=<5 0 1.5>, rotation=<rot 0 0 90>}]
 dyn_hair_bk_l_03 [{shapeType=Capsule, length=8.7957, radius=5.9543, shapeName=dyn_hair_bk_01_capsule, center=<5 -2 4>, rotation=<rot -0.0001 0 85>}]
 dyn_hair_bk_l_MSTR [{shapeType=Capsule, length=0.505, radius=0.505, shapeName=dyn_hair_bk_MSTR_capsule, rotation=<rot 0 -90 0>}]
 dyn_hair_bk_r_01 [{shapeType=Capsule, length=4, radius=6, shapeName=dyn_hair_bk_01_capsule, center=<5 0 4>, rotation=<rot 84.9999 -0.0001 -0.0002>}]
 dyn_hair_bk_r_02 [{shapeType=Capsule, length=3.6993, radius=6.0267, shapeName=dyn_hair_bk_01_capsule, center=<5 0 1.5>, rotation=<rot -0.0002 0 75>}]
 dyn_hair_bk_r_03 [{shapeType=Capsule, length=10.4188, radius=5.9543, shapeName=dyn_hair_bk_01_capsule, center=<5 3 4>, rotation=<rot -0.0005 0 90.0001>}]
 dyn_hair_bk_r_MSTR [{shapeType=Capsule, length=0.505, radius=0.505, shapeName=dyn_hair_bk_MSTR_capsule, rotation=<rot 0 -90 0>}]
 head [{shapeType=Capsule, length=4.25, radius=7.375, shapeName=spine_05_capsule, center=<2.8886 2.2093 0>, rotation=<rot 90 176.4237 167.1363>}]
 neck_01 [{shapeType=Capsule, length=2.6342, radius=7.7408, shapeName=spine_05_capsule, center=<4.0344 3.1739 0>, rotation=<rot 90 180 -178.816>}]
 spine_05 [{shapeType=Capsule, length=11.1701, radius=13.625, shapeName=spine_05_capsule, center=<0.2697 3.4051 0>}]
 upperarm_l [{shapeType=Capsule, length=15.625, radius=5.25, shapeName=clavicle_l_capsule, center=<8.7605 -0.497 -1.0162>, rotation=<rot 90 0 180>}]
 upperarm_r [{shapeType=Capsule, length=18.327, radius=5.25, shapeName=clavicle_l_capsule, center=<-9.5806 0.497 1.0162>, rotation=<rot 90 0 0>}]
}
constraints [
 {
  bone1Name=dyn_hair_bk_MSTR
  bone2Name=head
  swing1LimitDegrees=45
  swing1Motion=Free
  swing2LimitDegrees=45
  swing2Motion=Free
  twistLimitDegrees=45
  twistMotion=Free
 }
 {bone1Name=dyn_hair_bk_01, bone2Name=dyn_hair_bk_MSTR, swing1LimitDegrees=12, swing2LimitDegrees=60, twistLimitDegrees=2}
 {
  bone1Name=dyn_hair_bk_l_MSTR
  bone2Name=head
  swing1LimitDegrees=45
  swing1Motion=Free
  swing2LimitDegrees=45
  swing2Motion=Free
  twistLimitDegrees=45
  twistMotion=Free
 }
 {bone1Name=dyn_hair_bk_l_01, bone2Name=dyn_hair_bk_l_MSTR, swing1LimitDegrees=12, swing2LimitDegrees=60, twistLimitDegrees=2}
 {
  bone1Name=head
  bone2Name=neck_01
  swing1LimitDegrees=45
  swing1Motion=Free
  swing2LimitDegrees=45
  swing2Motion=Free
  twistLimitDegrees=45
  twistMotion=Free
 }
 {bone1Name=dyn_hair_bk_l_02, bone2Name=dyn_hair_bk_l_01, swing1LimitDegrees=12, swing2LimitDegrees=60, twistLimitDegrees=15}
 {bone1Name=dyn_hair_bk_02, bone2Name=dyn_hair_bk_01, swing1LimitDegrees=12, swing2LimitDegrees=60, twistLimitDegrees=15}
 {
  bone1Name=neck_01
  bone2Name=spine_05
  swing1LimitDegrees=45
  swing1Motion=Free
  swing2LimitDegrees=45
  swing2Motion=Free
  twistLimitDegrees=45
  twistMotion=Free
 }
 {bone1Name=dyn_hair_bk_l_03, bone2Name=dyn_hair_bk_l_02, swing1LimitDegrees=12, swing2LimitDegrees=60, twistLimitDegrees=15}
 {bone1Name=dyn_hair_bk_03, bone2Name=dyn_hair_bk_02, swing1LimitDegrees=12, swing2LimitDegrees=60, twistLimitDegrees=15}
 {bone1Name=dyn_hair_bk_r_01, bone2Name=dyn_hair_bk_r_MSTR, swing1LimitDegrees=12, swing2LimitDegrees=60, twistLimitDegrees=2}
 {
  bone1Name=dyn_hair_bk_r_MSTR
  bone2Name=head
  swing1LimitDegrees=45
  swing1Motion=Free
  swing2LimitDegrees=45
  swing2Motion=Free
  twistLimitDegrees=45
  twistMotion=Free
 }
 {bone1Name=dyn_hair_bk_r_02, bone2Name=dyn_hair_bk_r_01, swing1LimitDegrees=12, swing2LimitDegrees=60, twistLimitDegrees=15}
 {bone1Name=dyn_hair_bk_r_03, bone2Name=dyn_hair_bk_r_02, swing1LimitDegrees=12, swing2LimitDegrees=60, twistLimitDegrees=15}
 {bone1Name=clavicle_r, bone2Name=spine_05, swing1LimitDegrees=20, swing2LimitDegrees=25, twistLimitDegrees=2}
 {
  bone1Name=upperarm_r
  bone2Name=clavicle_r
  swing1LimitDegrees=45
  swing1Motion=Free
  swing2LimitDegrees=45
  swing2Motion=Free
  twistLimitDegrees=45
  twistMotion=Free
 }
 {
  bone1Name=upperarm_l
  bone2Name=clavicle_l
  swing1LimitDegrees=45
  swing1Motion=Free
  swing2LimitDegrees=45
  swing2Motion=Free
  twistLimitDegrees=45
  twistMotion=Free
 }
 {bone1Name=clavicle_l, bone2Name=spine_05, swing1LimitDegrees=20, swing2LimitDegrees=25, twistLimitDegrees=2}
]
