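# Post-process style AnimBP for Sparrow's face-accessory mesh, animating back hair, hair tips, earrings and lip ring over the copied body pose.
# A cached base pose constrains lip bones to facial bones, offsets the ring by lip distance, and runs AnimDynamics on earrings and tips.
# Back hair uses RigidBody, AnimDynamics when wearing backbling, or procedural-only posing when rigid bodies are disabled.
# Look-ats and head-twist-driven corrective offsets shape these branches.
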AnimBlueprint /BRCosmetics/Characters/Player/Male/Medium/Bodies/M_MED_Sparrow/Meshes/Parts/M_MED_Sparrow_FaceAcc_AnimBP
target_skeleton=&/M_MED_Sparrow_FaceAcc_Skeleton
graphs [
 AnimGraph {nodes [
   n1 = Look At - Bone: dyn_hair_bk_MSTR(ComponentPose=Local To Component(LocalPose=Use cached pose 'BasePose'()), Alpha=1) {
    alphaBoolBlend {blendOption=Linear}
    boneToModify {boneName=dyn_hair_bk_MSTR}
    lookAtLocation=<-12 -12.5>
    lookAtTarget {boneReference {boneName=spine_05}}
    lookAt_Axis {axis=<1 0>}
   }
   n2 = Look At - Bone: dyn_hair_bk_r_MSTR(ComponentPose=@n1, Alpha=1) {
    alphaBoolBlend {blendOption=Linear}
    boneToModify {boneName=dyn_hair_bk_r_MSTR}
    lookAtLocation=<-12 -16 7.8>
    lookAtTarget {boneReference {boneName=spine_05}}
    lookAt_Axis {axis=<1 0>}
   }
   n3 = Look At - Bone: dyn_hair_bk_l_MSTR(ComponentPose=@n2, Alpha=1) {
    alphaBoolBlend {blendOption=Linear}
    boneToModify {boneName=dyn_hair_bk_l_MSTR}
    lookAtLocation=<-12 -16 -7.8>
    lookAtTarget {boneReference {boneName=spine_05}}
    lookAt_Axis {axis=<1 0>}
   }
   n4 = Get head_rot_bk() -> (head_rot_bk:double)
   n5 = "Transform (Modify) Bone - Bone: dyn_hair_bk_r_MSTR"(ComponentPose=@n3, Alpha=@n4, Translation="0,0,3", Rotation="0,0,0", Scale="1,1,1") {
    translationMode=BMM_Additive
    translationSpace=BCS_BoneSpace
    alphaBoolBlend {blendOption=Linear}
    boneToModify {boneName=dyn_hair_bk_r_MSTR}
    translation {z=3}
   }
   n6 = "Transform (Modify) Bone - Bone: dyn_hair_bk_l_MSTR"(ComponentPose=@n5, Alpha=@n4, Translation="0,0,3", Rotation="0,0,0", Scale="1,1,1") {
    translationMode=BMM_Additive
    translationSpace=BCS_BoneSpace
    alphaBoolBlend {blendOption=Linear}
    boneToModify {boneName=dyn_hair_bk_l_MSTR}
    translation {z=3}
   }
   n7 = "Transform (Modify) Bone - Bone: dyn_hair_bk_MSTR"(ComponentPose=@n6, Alpha=@n4, Translation="0,0,1", Rotation="0,0,0", Scale="1,1,1") {
    rotationMode=BMM_Additive
    rotationSpace=BCS_BoneSpace
    translationMode=BMM_Additive
    translationSpace=BCS_BoneSpace
    alphaBoolBlend {blendOption=Linear}
    boneToModify {boneName=dyn_hair_bk_MSTR}
    translation {z=1}
   }
   n8 = Get head_rot_r() -> (head_rot_r:double)
   n9 = "Transform (Modify) Bone - Bone: dyn_hair_bk_l_MSTR"(ComponentPose=@n7, Alpha=@n8, Translation="4,0,2", Rotation="0,-20,0", Scale="1,1,1") {
    rotationMode=BMM_Additive
    rotationSpace=BCS_BoneSpace
    translationMode=BMM_Additive
    translationSpace=BCS_BoneSpace
    alphaBoolBlend {blendOption=Linear}
    boneToModify {boneName=dyn_hair_bk_l_MSTR}
    rotation {yaw=-20}
    translation {x=4, z=2}
   }
   n10 = "Transform (Modify) Bone - Bone: dyn_hair_bk_MSTR"(ComponentPose=@n9, Alpha=@n8, Translation="2,0,3", Rotation="0,-20,0", Scale="1,1,1") {
    rotationMode=BMM_Additive
    rotationSpace=BCS_BoneSpace
    translationMode=BMM_Additive
    translationSpace=BCS_BoneSpace
    alphaBoolBlend {blendOption=Linear}
    boneToModify {boneName=dyn_hair_bk_MSTR}
    rotation {yaw=-20}
    translation {x=2, z=3}
   }
   n11 = "Transform (Modify) Bone - Bone: dyn_hair_bk_r_MSTR"(ComponentPose=@n10, Alpha=@n8, Translation="0,-2,8", Rotation="0,-15,0", Scale="1,1,1") {
    rotationMode=BMM_Additive
    rotationSpace=BCS_BoneSpace
    translationMode=BMM_Additive
    translationSpace=BCS_BoneSpace
    alphaBoolBlend {blendOption=Linear}
    boneToModify {boneName=dyn_hair_bk_r_MSTR}
    rotation {yaw=-15}
    translation {y=-2, z=8}
   }
   n12 = RigidBody!(ComponentPose=@n11, Alpha, OverrideWorldGravity:Vector="0,0,-600", ExternalForce, ComponentLinearAccScale:Vector="0.2,0.2,0.3") -> (Pose) {
    bClampLinearTranslationLimitToRefPose=true
    bOverrideWorldGravity=true
    overridePhysicsAsset=&/M_MED_Sparrow_FaceAcc_FaceAcc_Dyn_Hair
    simulationSpace=BaseBoneSpace
    alphaBoolBlend {blendOption=Linear}
    alphaScaleBiasClamp {interpSpeedDecreasing=4, interpSpeedIncreasing=6}
    baseBoneRef {boneName=spine_05}
    componentAppliedLinearAccClamp=<3000 3000 3000>
    componentLinearAccScale=<0.6 0.6 0.2>
    componentLinearVelScale=<0.35 0.35 0.35>
    overrideWorldGravity {z=-875}
    simSpaceSettings {maxAngularAcceleration=200, maxAngularVelocity=5, maxLinearAcceleration=500, maxLinearVelocity=200}
   }
   n13 = Component To Local(ComponentPose=@n12)
   n14 = Look At - Bone: dyn_hair_bk_01(ComponentPose=Local To Component(LocalPose=Use cached pose 'BasePose'()), Alpha=1) {
    alphaBoolBlend {blendOption=Linear}
    boneToModify {boneName=dyn_hair_bk_01}
    lookAtLocation=<-12 -12.5>
    lookAtTarget {boneReference {boneName=spine_05}}
    lookAt_Axis {axis=<1 0>}
   }
   n15 = Look At - Bone: dyn_hair_bk_r_01(ComponentPose=@n14, Alpha=1) {
    alphaBoolBlend {blendOption=Linear}
    boneToModify {boneName=dyn_hair_bk_r_01}
    lookAtLocation=<-12 -16 7.8>
    lookAtTarget {boneReference {boneName=spine_05}}
    lookAt_Axis {axis=<1 0>}
   }
   n16 = Look At - Bone: dyn_hair_bk_l_01(ComponentPose=@n15, Alpha=1) {
    alphaBoolBlend {blendOption=Linear}
    boneToModify {boneName=dyn_hair_bk_l_01}
    lookAtLocation=<-12 -16 -7.8>
    lookAtTarget {boneReference {boneName=spine_05}}
    lookAt_Axis {axis=<1 0>}
   }
   n17 = "Constraint - dyn_hair_bk_02 (P)"(ComponentPose=@n16, Alpha=1, ConstraintWeights_0=1) {
    alphaBoolBlend {blendOption=Linear}
    boneToModify {boneName=dyn_hair_bk_02}
    constraintSetup [{targetBone {boneName=spine_05}}]
    constraintWeights [1]
   }
   n18 = Look At - Bone: dyn_hair_bk_02(ComponentPose=@n17, Alpha=1) {
    alphaBoolBlend {blendOption=Linear}
    boneToModify {boneName=dyn_hair_bk_02}
    lookAtLocation=<-12 -21>
    lookAtTarget {boneReference {boneName=spine_05}}
    lookAt_Axis {axis=<1 0>}
   }
   n19 = "Constraint - dyn_hair_bk_r_02 (PP)"!(ComponentPose=@n18, Alpha=1, ConstraintWeights_0=0.8, ConstraintWeights_1:float=0.2) -> (Pose) {
    alphaBoolBlend {blendOption=Linear}
    boneToModify {boneName=dyn_hair_bk_r_02}
    constraintSetup [{targetBone {boneName=spine_05}}, {targetBone {boneName=clavicle_r}}]
    constraintWeights [1,1]
   }
   n20 = Look At - Bone: dyn_hair_bk_r_02(ComponentPose=@n19, Alpha=1) {
    alphaBoolBlend {blendOption=Linear}
    boneToModify {boneName=dyn_hair_bk_r_02}
    lookAtLocation=<-12 -22 11>
    lookAtTarget {boneReference {boneName=spine_05}}
    lookAt_Axis {axis=<1 0>}
   }
   n21 = "Constraint - dyn_hair_bk_l_02 (PP)"!(ComponentPose=@n20, Alpha=1, ConstraintWeights_0=0.7, ConstraintWeights_1:float=0.3) -> (Pose) {
    alphaBoolBlend {blendOption=Linear}
    boneToModify {boneName=dyn_hair_bk_l_02}
    constraintSetup [{targetBone {boneName=spine_05}}, {targetBone {boneName=clavicle_l}}]
    constraintWeights [1,1]
   }
   n22 = Look At - Bone: dyn_hair_bk_l_02(ComponentPose=@n21, Alpha=1) {
    alphaBoolBlend {blendOption=Linear}
    boneToModify {boneName=dyn_hair_bk_l_02}
    lookAtLocation=<-12 -24 -11>
    lookAtTarget {boneReference {boneName=spine_05}}
    lookAt_Axis {axis=<1 0>}
   }
   n23 = Get head_rot_bk()
   n24 = "Transform (Modify) Bone - Bone: dyn_hair_bk_r_02"(ComponentPose=@n22, Alpha=@n23, Translation="4,0,2", Rotation="-15,0,0", Scale="1,1,1") {
    rotationMode=BMM_Additive
    rotationSpace=BCS_BoneSpace
    translationMode=BMM_Additive
    translationSpace=BCS_BoneSpace
    alphaBoolBlend {blendOption=Linear}
    boneToModify {boneName=dyn_hair_bk_r_02}
    rotation {pitch=-15}
    translation {x=4, z=2}
   }
   n25 = "Transform (Modify) Bone - Bone: dyn_hair_bk_l_02"(ComponentPose=@n24, Alpha=@n23, Translation="4,0,2", Rotation="-20,0,0", Scale="1,1,1") {
    rotationMode=BMM_Additive
    rotationSpace=BCS_BoneSpace
    translationMode=BMM_Additive
    translationSpace=BCS_BoneSpace
    alphaBoolBlend {blendOption=Linear}
    boneToModify {boneName=dyn_hair_bk_l_02}
    rotation {pitch=-20}
    translation {x=4, z=2}
   }
   n26 = "Transform (Modify) Bone - Bone: dyn_hair_bk_02"(ComponentPose=@n25, Alpha=@n23, Translation="6,0,3", Rotation="-20,0,0", Scale="1,1,1") {
    rotationMode=BMM_Additive
    rotationSpace=BCS_BoneSpace
    translationMode=BMM_Additive
    translationSpace=BCS_BoneSpace
    alphaBoolBlend {blendOption=Linear}
    boneToModify {boneName=dyn_hair_bk_02}
    rotation {pitch=-20}
    translation {x=6, z=3}
   }
   n27 = Get head_rot_r()
   n28 = "Transform (Modify) Bone - Bone: dyn_hair_bk_l_02"(ComponentPose=@n26, Alpha=@n27, Translation="0,-3,0", Rotation="0,-20,0", Scale="1,1,1") {
    rotationMode=BMM_Additive
    rotationSpace=BCS_BoneSpace
    translationMode=BMM_Additive
    translationSpace=BCS_BoneSpace
    alphaBoolBlend {blendOption=Linear}
    boneToModify {boneName=dyn_hair_bk_l_02}
    rotation {yaw=-20}
    translation {y=-3}
   }
   n29 = "Transform (Modify) Bone - Bone: dyn_hair_bk_02"(ComponentPose=@n28, Alpha=@n27, Translation="4,-8,0", Rotation="0,-20,0", Scale="1,1,1") {
    rotationMode=BMM_Additive
    rotationSpace=BCS_BoneSpace
    translationMode=BMM_Additive
    translationSpace=BCS_BoneSpace
    alphaBoolBlend {blendOption=Linear}
    boneToModify {boneName=dyn_hair_bk_02}
    rotation {yaw=-20}
    translation=<4 -8>
   }
   n30 = "Transform (Modify) Bone - Bone: dyn_hair_bk_r_MSTR"(ComponentPose=@n29, Alpha=@n27, Translation="-3,0,4", Rotation="0,0,0", Scale="1,1,1") {
    rotationMode=BMM_Additive
    rotationSpace=BCS_BoneSpace
    translationMode=BMM_Additive
    translationSpace=BCS_BoneSpace
    alphaBoolBlend {blendOption=Linear}
    boneToModify {boneName=dyn_hair_bk_r_MSTR}
    translation {x=-3, z=4}
   }
   n31 = "Transform (Modify) Bone - Bone: dyn_hair_bk_r_02"(ComponentPose=@n30, Alpha=@n27, Translation="7,-8,-4", Rotation="0,-15,0", Scale="1,1,1") {
    rotationMode=BMM_Additive
    rotationSpace=BCS_BoneSpace
    translationMode=BMM_Additive
    translationSpace=BCS_BoneSpace
    alphaBoolBlend {blendOption=Linear}
    boneToModify {boneName=dyn_hair_bk_r_02}
    rotation {yaw=-15}
    translation=<7 -8 -4>
   }
   n32 = "Anim Dynamics - Chain: dyn_hair_bk_02 -> dyn_hair_bk_03"!(ComponentPose=@n31, Alpha=1, GravityScale:float=1, ExternalForce="0,0,0") -> (Pose) {
    angularDampingOverride=0.9999
    angularSpringConstant=600
    bAngularSpring=true
    bChain=true
    bGravityOverrideInSimSpace=true
    bOverrideAngularDamping=true
    bUseGravityOverride=true
    simulationSpace=BoneRelative
    alphaBoolBlend {blendOption=Linear}
    boundBone {boneName=dyn_hair_bk_02}
    chainEnd {boneName=dyn_hair_bk_03}
    componentAppliedLinearAccClamp=<1000 1000 1000>
    componentLinearAccScale=<0.2 0.2 0.1>
    componentLinearVelScale=<0.2 0.2 0.2>
    gravityOverride {x=-980}
    physicsBodyDefinitions [
     {
      sphereCollisionRadius=0
      boundBone {boneName=dyn_hair_bk_02}
      boxExtents=<15 10 10>
      constraintSetup {angularLimitsMax=<0 0 20>, angularLimitsMin=<0 0 -20>, angularTarget=<0 0 0>}
      localJointOffset=<3 0 0>
     }
     {
      sphereCollisionRadius=0
      boundBone {boneName=dyn_hair_bk_03}
      boxExtents=<15 10 10>
      constraintSetup {angularLimitsMax=<0 15 20>, angularLimitsMin=<0 0 -20>, angularTarget=<0 0 0>}
      localJointOffset=<1 0 -1>
     }
    ]
    relativeSpaceBone {boneName=neck_02}
   }
   n33 = "Anim Dynamics - Chain: dyn_hair_bk_r_02 -> dyn_hair_bk_r_03"!(ComponentPose=@n32, Alpha=1, GravityScale:float=1, ExternalForce="0,0,0") -> (Pose) {
    angularDampingOverride=0.9999
    angularSpringConstant=600
    bAngularSpring=true
    bChain=true
    bGravityOverrideInSimSpace=true
    bOverrideAngularDamping=true
    bUseGravityOverride=true
    simulationSpace=BoneRelative
    alphaBoolBlend {blendOption=Linear}
    boundBone {boneName=dyn_hair_bk_r_02}
    chainEnd {boneName=dyn_hair_bk_r_03}
    componentAppliedLinearAccClamp=<1000 1000 1000>
    componentLinearAccScale=<0.2 0.2 0.1>
    componentLinearVelScale=<0.2 0.2 0.2>
    gravityOverride {x=-980}
    physicsBodyDefinitions [
     {
      sphereCollisionRadius=0
      boundBone {boneName=dyn_hair_bk_r_02}
      boxExtents=<15 10 10>
      constraintSetup {angularLimitsMax=<0 0 20>, angularLimitsMin=<0 0 -30>, angularTarget=<0 0 0>}
      localJointOffset=<3 0 0>
     }
     {
      sphereCollisionRadius=0
      boundBone {boneName=dyn_hair_bk_r_03}
      boxExtents=<15 10 10>
      constraintSetup {angularLimitsMax=<0 15 20>, angularLimitsMin=<0 0 -30>, angularTarget=<0 0 0>}
      localJointOffset=<1 0 -1>
     }
    ]
    relativeSpaceBone {boneName=neck_02}
   }
   n34 = "Anim Dynamics - Chain: dyn_hair_bk_l_02 -> dyn_hair_bk_l_03"!(ComponentPose=@n33, Alpha=1, GravityScale:float=1, ExternalForce="0,0,0") -> (Pose) {
    angularDampingOverride=0.9999
    angularSpringConstant=600
    bAngularSpring=true
    bChain=true
    bGravityOverrideInSimSpace=true
    bOverrideAngularDamping=true
    bUseGravityOverride=true
    simulationSpace=BoneRelative
    alphaBoolBlend {blendOption=Linear}
    boundBone {boneName=dyn_hair_bk_l_02}
    chainEnd {boneName=dyn_hair_bk_l_03}
    componentAppliedLinearAccClamp=<1000 1000 1000>
    componentLinearAccScale=<0.2 0.2 0.1>
    componentLinearVelScale=<0.2 0.2 0.2>
    gravityOverride {x=-980}
    physicsBodyDefinitions [
     {
      sphereCollisionRadius=0
      boundBone {boneName=dyn_hair_bk_l_02}
      boxExtents=<15 10 10>
      constraintSetup {angularLimitsMax=<0 0 30>, angularLimitsMin=<0 0 -20>, angularTarget=<0 0 0>}
      localJointOffset=<3 0 0>
     }
     {
      sphereCollisionRadius=0
      boundBone {boneName=dyn_hair_bk_l_03}
      boxExtents=<15 10 10>
      constraintSetup {angularLimitsMax=<0 15 30>, angularLimitsMin=<0 0 -20>, angularTarget=<0 0 0>}
      localJointOffset=<1 0 -1>
     }
    ]
    relativeSpaceBone {boneName=neck_02}
   }
   n35 = Component To Local(ComponentPose=@n34)
   n36 = Get isWearBackbling() -> (isWearBackbling:bool)
   n37 = Look At - Bone: dyn_hair_bk_01(ComponentPose=Local To Component(LocalPose=Use cached pose 'BasePose'()), Alpha=1) {
    alphaBoolBlend {blendOption=Linear}
    boneToModify {boneName=dyn_hair_bk_01}
    lookAtLocation=<-12 -12.5>
    lookAtTarget {boneReference {boneName=spine_05}}
    lookAt_Axis {axis=<1 0>}
   }
   n38 = Look At - Bone: dyn_hair_bk_r_01(ComponentPose=@n37, Alpha=1) {
    alphaBoolBlend {blendOption=Linear}
    boneToModify {boneName=dyn_hair_bk_r_01}
    lookAtLocation=<-12 -16 7.8>
    lookAtTarget {boneReference {boneName=spine_05}}
    lookAt_Axis {axis=<1 0>}
   }
   n39 = Look At - Bone: dyn_hair_bk_l_01(ComponentPose=@n38, Alpha=1) {
    alphaBoolBlend {blendOption=Linear}
    boneToModify {boneName=dyn_hair_bk_l_01}
    lookAtLocation=<-12 -16 -7.8>
    lookAtTarget {boneReference {boneName=spine_05}}
    lookAt_Axis {axis=<1 0>}
   }
   n40 = Look At - Bone: dyn_hair_bk_02(ComponentPose=@n39, Alpha=1) {
    alphaBoolBlend {blendOption=Linear}
    boneToModify {boneName=dyn_hair_bk_02}
    lookAtLocation=<-12 -21>
    lookAtTarget {boneReference {boneName=spine_05}}
    lookAt_Axis {axis=<1 0>}
   }
   n41 = "Constraint - dyn_hair_bk_03 (P)"(ComponentPose=@n40, Alpha=1, ConstraintWeights_0=1) {
    alphaBoolBlend {blendOption=Linear}
    boneToModify {boneName=dyn_hair_bk_03}
    constraintSetup [{targetBone {boneName=spine_05}}]
    constraintWeights [1]
   }
   n42 = "Constraint - dyn_hair_bk_r_02 (PP)"(ComponentPose=@n41, Alpha=1, ConstraintWeights_0=0.8, ConstraintWeights_1=0.2) {
    alphaBoolBlend {blendOption=Linear}
    boneToModify {boneName=dyn_hair_bk_r_02}
    constraintSetup [{targetBone {boneName=spine_05}}, {targetBone {boneName=clavicle_r}}]
    constraintWeights [1,1]
   }
   n43 = Look At - Bone: dyn_hair_bk_r_02(ComponentPose=@n42, Alpha=1) {
    alphaBoolBlend {blendOption=Linear}
    boneToModify {boneName=dyn_hair_bk_r_02}
    lookAtLocation=<-12 -22 11>
    lookAtTarget {boneReference {boneName=spine_05}}
    lookAt_Axis {axis=<1 0>}
   }
   n44 = "Constraint - dyn_hair_bk_l_02 (PP)"(ComponentPose=@n43, Alpha=1, ConstraintWeights_0=0.8, ConstraintWeights_1=0.2) {
    alphaBoolBlend {blendOption=Linear}
    boneToModify {boneName=dyn_hair_bk_l_02}
    constraintSetup [{targetBone {boneName=spine_05}}, {targetBone {boneName=clavicle_l}}]
    constraintWeights [1,1]
   }
   n45 = Look At - Bone: dyn_hair_bk_l_02(ComponentPose=@n44, Alpha=1) {
    alphaBoolBlend {blendOption=Linear}
    boneToModify {boneName=dyn_hair_bk_l_02}
    lookAtLocation=<-12 -22 -11>
    lookAtTarget {boneReference {boneName=spine_05}}
    lookAt_Axis {axis=<1 0>}
   }
   n46 = Get head_rot_bk()
   n47 = "Transform (Modify) Bone - Bone: dyn_hair_bk_r_02"(ComponentPose=@n45, Alpha=@n46, Translation="3,0,0", Rotation="0,0,0", Scale="1,1,1") {
    translationMode=BMM_Additive
    translationSpace=BCS_BoneSpace
    alphaBoolBlend {blendOption=Linear}
    boneToModify {boneName=dyn_hair_bk_r_02}
    translation {x=3}
   }
   n48 = "Transform (Modify) Bone - Bone: dyn_hair_bk_l_02"(ComponentPose=@n47, Alpha=@n46, Translation="3,0,0", Rotation="0,0,0", Scale="1,1,1") {
    translationMode=BMM_Additive
    translationSpace=BCS_BoneSpace
    alphaBoolBlend {blendOption=Linear}
    boneToModify {boneName=dyn_hair_bk_l_02}
    translation {x=3}
   }
   n49 = "Transform (Modify) Bone - Bone: dyn_hair_bk_03"(ComponentPose=@n48, Alpha=@n46, Translation="3,0,0", Rotation="0,0,0", Scale="1,1,1") {
    rotationMode=BMM_Additive
    rotationSpace=BCS_BoneSpace
    translationMode=BMM_Additive
    translationSpace=BCS_BoneSpace
    alphaBoolBlend {blendOption=Linear}
    boneToModify {boneName=dyn_hair_bk_03}
    translation {x=3}
   }
   Output Pose(Result=Blend Poses by bool(
     BlendPose_0=Blend Poses by bool(BlendPose_0=@n35, BlendPose_1=@n13, bActiveValue=@n36) {blendType=Linear}
     BlendPose_1=Component To Local(ComponentPose=@n49)
     bActiveValue=Get bRigidBodyEnabled()
    ) {blendType=Linear}) {layerGroup=None}
   n50 = Copy Pose From Mesh(SourceMeshComponent=Get SkelMesh())
   n51 = Copy Pose From Mesh() {bUseAttachedParent=true}
   n52 = "Constraint - def_uppper_lip_PST (PPPPP)"!(
    ComponentPose=Local To Component(LocalPose=Layered blend per bone(BasePose=@n51, BlendPoses_0=@n50) {layerSetup [{branchFilters [{boneName=upperarm_l}, {boneName=upperarm_r}]}]})
    Alpha=1
    ConstraintWeights_0=0.0168
    ConstraintWeights_1:float=0.4069
    ConstraintWeights_2:float=0.4535
    ConstraintWeights_3:float=0.0671
    ConstraintWeights_4:float=0.0558
   ) -> (Pose) {
    alphaBoolBlend {blendOption=Linear}
    boneToModify {boneName=def_uppper_lip_PST}
    constraintSetup [
     {targetBone {boneName=head}}
     {targetBone {boneName=FACIAL_C_LipUpper}}
     {targetBone {boneName=FACIAL_R_LipUpper}}
     {targetBone {boneName=FACIAL_R_LipUpperOuter}}
     {targetBone {boneName=FACIAL_R_LipCorner}}
    ]
    constraintWeights [1,1,1,1,1]
   }
   n53 = "Constraint - def_lower_lip_PST (PPPPP)"!(
    ComponentPose=@n52
    Alpha=1
    ConstraintWeights_0=0.0292
    ConstraintWeights_1:float=0.0085
    ConstraintWeights_2:float=0.4774
    ConstraintWeights_3:float=0.4394
    ConstraintWeights_4:float=0.0455
   ) -> (Pose) {
    alphaBoolBlend {blendOption=Linear}
    boneToModify {boneName=def_lower_lip_PST}
    constraintSetup [
     {targetBone {boneName=FACIAL_R_LipCorner}}
     {targetBone {boneName=FACIAL_C_LowerLipRotation}}
     {targetBone {boneName=FACIAL_C_LipLower}}
     {targetBone {boneName=FACIAL_R_LipLower}}
     {targetBone {boneName=FACIAL_R_LipLowerOuter}}
    ]
    constraintWeights [1,1,1,1,1]
   }
   n54 = "Constraint - def_lip_ring (PPPPP)"!(
    ComponentPose=@n53
    Alpha=1
    ConstraintWeights_0=0.028
    ConstraintWeights_1:float=0.5017
    ConstraintWeights_2:float=0.4232
    ConstraintWeights_3:float=0.0438
    ConstraintWeights_4:float=0.0033
   ) -> (Pose) {
    alphaBoolBlend {blendOption=Linear}
    boneToModify {boneName=def_lip_ring}
    constraintSetup [
     {targetBone {boneName=FACIAL_R_LipCorner}}
     {targetBone {boneName=FACIAL_C_LipLower}}
     {targetBone {boneName=FACIAL_R_LipLower}}
     {targetBone {boneName=FACIAL_R_LipLowerOuter}}
     {targetBone {boneName=FACIAL_C_Jaw}}
    ]
    constraintWeights [1,1,1,1,1]
   }
   n55 = "Transform (Modify) Bone - Bone: def_lip_ring"(
    ComponentPose=@n54
    Alpha=Get movingUpLipRingAlpha() -> (movingUpLipRingAlpha:double)
    Translation="0,0,0.09"
    Rotation="0,0,0"
    Scale="1,1,1"
   ) {
    translationMode=BMM_Additive
    translationSpace=BCS_BoneSpace
    alphaBoolBlend {blendOption=Linear}
    boneToModify {boneName=def_lip_ring}
    translation {z=0.09}
   }
   n56 = Anim Dynamics - Bone: dyn_earring_l(ComponentPose=@n55, Alpha=1, ExternalForce="0,0,0") {
    angularDampingOverride=0.999
    angularSpringConstant=1000
    bAngularSpring=true
    bOverrideAngularDamping=true
    bUseGravityOverride=true
    simulationSpace=RootRelative
    alphaBoolBlend {blendOption=Linear}
    boundBone {boneName=dyn_earring_l}
    componentAppliedLinearAccClamp=<5000 5000 5000>
    componentLinearAccScale=<0.3 0.3 0.3>
    componentLinearVelScale=<0.3 0.3 0.3>
    gravityOverride {z=-980}
    physicsBodyDefinitions [{
      boundBone {boneName=dyn_earring_l}
      boxExtents=<20 10 10>
      constraintSetup {angularLimitsMax=<0 15 45>, angularLimitsMin=<0 -35 -45>, angularTarget=<0 0 0>}
      localJointOffset=<5 0 0>
     }]
   }
   n57 = Anim Dynamics - Bone: dyn_earring_r(ComponentPose=@n56, Alpha=1, ExternalForce="0,0,0") {
    angularDampingOverride=0.999
    angularSpringConstant=1000
    bAngularSpring=true
    bOverrideAngularDamping=true
    bUseGravityOverride=true
    simulationSpace=RootRelative
    alphaBoolBlend {blendOption=Linear}
    boundBone {boneName=dyn_earring_r}
    componentAppliedLinearAccClamp=<5000 5000 5000>
    componentLinearAccScale=<0.3 0.3 0.3>
    componentLinearVelScale=<0.3 0.3 0.3>
    gravityOverride {z=-980}
    physicsBodyDefinitions [{
      boundBone {boneName=dyn_earring_r}
      boxExtents=<20 10 10>
      constraintSetup {angularLimitsMax=<0 15 45>, angularLimitsMin=<0 -35 -45>, angularTarget=<0 0 0>}
      localJointOffset=<5 0 0>
     }]
   }
   n58 = "Anim Dynamics - Chain: dyn_hairTips_A_01 -> dyn_hairTips_A_02"!(ComponentPose=@n57, Alpha=1, GravityScale:float=1, ExternalForce="0,0,0") -> (Pose) {
    angularDampingOverride=0.99
    angularSpringConstant=800
    bAngularSpring=true
    bChain=true
    bGravityOverrideInSimSpace=true
    bOverrideAngularDamping=true
    bUseGravityOverride=true
    simulationSpace=BoneRelative
    alphaBoolBlend {blendOption=Linear}
    boundBone {boneName=dyn_hairTips_A_01}
    chainEnd {boneName=dyn_hairTips_A_02}
    componentAppliedLinearAccClamp=<1000 1000 1000>
    componentLinearAccScale=<0.2 0.2 0.2>
    componentLinearVelScale=<0.2 0.2 0.2>
    gravityOverride {x=-980}
    physicsBodyDefinitions [
     {
      sphereCollisionRadius=0
      boundBone {boneName=dyn_hairTips_A_01}
      boxExtents=<15 10 10>
      constraintSetup {angularLimitsMax=<0 0 30>, angularLimitsMin=<0 -15 -35>, angularTarget=<0 0 0>}
      localJointOffset=<3 -0.5 1>
     }
     {
      sphereCollisionRadius=0
      boundBone {boneName=dyn_hairTips_A_02}
      boxExtents=<15 10 10>
      constraintSetup {angularLimitsMax=<0 10 45>, angularLimitsMin=<0 -15 -35>, angularTarget=<0 0 0>}
      localJointOffset=<2 -2 0>
     }
    ]
    relativeSpaceBone {boneName=neck_02}
   }
   n59 = "Anim Dynamics - Chain: dyn_hairTips_B_01 -> dyn_hairTips_B_02"!(ComponentPose=@n58, Alpha=1, GravityScale:float=1, ExternalForce="0,0,0") -> (Pose) {
    angularDampingOverride=0.999
    angularSpringConstant=600
    bAngularSpring=true
    bChain=true
    bGravityOverrideInSimSpace=true
    bOverrideAngularDamping=true
    bUseGravityOverride=true
    simulationSpace=BoneRelative
    alphaBoolBlend {blendOption=Linear}
    boundBone {boneName=dyn_hairTips_B_01}
    chainEnd {boneName=dyn_hairTips_B_02}
    componentAppliedLinearAccClamp=<1000 1000 1000>
    componentLinearAccScale=<0.2 0.2 0.2>
    componentLinearVelScale=<0.2 0.2 0.2>
    gravityOverride {x=-980}
    physicsBodyDefinitions [
     {
      sphereCollisionRadius=0
      boundBone {boneName=dyn_hairTips_B_01}
      boxExtents=<10 15 15>
      constraintSetup {angularLimitsMax=<0 5 30>, angularLimitsMin=<0 -15 -30>, angularTarget=<0 0 0>}
      localJointOffset=<2 0.2 0>
     }
     {
      sphereCollisionRadius=0
      boundBone {boneName=dyn_hairTips_B_02}
      boxExtents=<10 10 10>
      constraintSetup {angularLimitsMax=<0 10 30>, angularLimitsMin=<0 -15 -30>, angularTarget=<0 0 0>}
      localJointOffset=<1 0.2 -1>
     }
    ]
    relativeSpaceBone {boneName=neck_01}
   }
   n60 = Anim Dynamics - Bone: dyn_hairTips_01!(ComponentPose=@n59, Alpha=1, GravityScale:float=1, ExternalForce="0,0,0") -> (Pose) {
    angularDampingOverride=0.99
    angularSpringConstant=800
    bAngularSpring=true
    bGravityOverrideInSimSpace=true
    bOverrideAngularDamping=true
    bUseGravityOverride=true
    simulationSpace=BoneRelative
    alphaBoolBlend {blendOption=Linear}
    boundBone {boneName=dyn_hairTips_01}
    componentAppliedLinearAccClamp=<1000 1000 1000>
    componentLinearAccScale=<0.2 0.2 0.2>
    componentLinearVelScale=<0.2 0.2 0.2>
    gravityOverride {x=-980}
    physicsBodyDefinitions [{
      sphereCollisionRadius=0
      boundBone {boneName=dyn_hairTips_01}
      boxExtents=<15 15 15>
      constraintSetup {angularLimitsMax=<0 20 20>, angularLimitsMin=<0 -20 -20>, angularTarget=<0 0 0>}
      localJointOffset=<-2 -1 -2>
     }]
    relativeSpaceBone {boneName=neck_01}
   }
   n61 = Anim Dynamics - Bone: dyn_hairTips_02!(ComponentPose=@n60, Alpha=1, GravityScale:float=1, ExternalForce="0,0,0") -> (Pose) {
    angularDampingOverride=0.993
    angularSpringConstant=800
    bAngularSpring=true
    bGravityOverrideInSimSpace=true
    bOverrideAngularDamping=true
    bUseGravityOverride=true
    simulationSpace=BoneRelative
    alphaBoolBlend {blendOption=Linear}
    boundBone {boneName=dyn_hairTips_02}
    componentAppliedLinearAccClamp=<1000 1000 1000>
    componentLinearAccScale=<0.1 0.1 0.1>
    componentLinearVelScale=<0.1 0.1 0.1>
    gravityOverride {x=-980}
    physicsBodyDefinitions [{
      sphereCollisionRadius=0
      boundBone {boneName=dyn_hairTips_02}
      boxExtents=<15 10 10>
      constraintSetup {angularLimitsMax=<0 5 45>, angularLimitsMin=<0 -15 -45>, angularTarget=<0 0 0>}
      localJointOffset=<2 2.2 -0.4>
     }]
    relativeSpaceBone {boneName=neck_01}
   }
   n62 = Anim Dynamics - Bone: dyn_hairTips_03!(ComponentPose=@n61, Alpha=1, GravityScale:float=1, ExternalForce="0,0,0") -> (Pose) {
    angularDampingOverride=0.995
    angularSpringConstant=800
    bAngularSpring=true
    bGravityOverrideInSimSpace=true
    bOverrideAngularDamping=true
    bUseGravityOverride=true
    simulationSpace=BoneRelative
    alphaBoolBlend {blendOption=Linear}
    boundBone {boneName=dyn_hairTips_03}
    componentAppliedLinearAccClamp=<1000 1000 1000>
    componentLinearAccScale=<0.1 0.1 0.1>
    componentLinearVelScale=<0.1 0.1 0.1>
    gravityOverride {x=-800}
    physicsBodyDefinitions [{
      sphereCollisionRadius=0
      boundBone {boneName=dyn_hairTips_03}
      boxExtents=<15 10 10>
      constraintSetup {angularLimitsMax=<0 5 35>, angularLimitsMin=<0 -15 -35>, angularTarget=<0 0 0>}
      localJointOffset=<1 1.5 0>
     }]
    relativeSpaceBone {boneName=neck_01}
   }
   n63 = Anim Dynamics - Bone: dyn_hairTips_04!(ComponentPose=@n62, Alpha=1, GravityScale:float=1, ExternalForce="0,0,0") -> (Pose) {
    angularDampingOverride=0.995
    angularSpringConstant=800
    bAngularSpring=true
    bGravityOverrideInSimSpace=true
    bOverrideAngularDamping=true
    bUseGravityOverride=true
    simulationSpace=BoneRelative
    alphaBoolBlend {blendOption=Linear}
    boundBone {boneName=dyn_hairTips_04}
    componentAppliedLinearAccClamp=<1000 1000 1000>
    componentLinearAccScale=<0.1 0.1 0.1>
    componentLinearVelScale=<0.1 0.1 0.1>
    gravityOverride {x=-800}
    physicsBodyDefinitions [{
      sphereCollisionRadius=0
      boundBone {boneName=dyn_hairTips_04}
      boxExtents=<15 10 10>
      constraintSetup {angularLimitsMax=<0 5 30>, angularLimitsMin=<0 -15 -30>, angularTarget=<0 0 0>}
      localJointOffset=<2 -2 0.1>
     }]
    relativeSpaceBone {boneName=neck_01}
   }
   Save cached pose 'BasePose'(Pose=Component To Local(ComponentPose=@n63)) {cacheName=BasePose}
  ]}
 EventGraph {nodes [
   Event BlueprintUpdateAnimation()
   n1 = Get SkelMesh()
   > Set head_rot_bk(head_rot_bk:double=GetAlphaFromReferenceTransformDelta(SkeletalMeshComponent=@n1, BoneName=head, BaseName=spine_05, DeltaComponentType=QuaternionTwist, InputRange=(X=0,Y=50), TwistAxis="0,0,1")) -> (Output_Get:double)
   > Set head_rot_r(head_rot_r:double=GetAlphaFromReferenceTransformDelta(SkeletalMeshComponent=@n1, BoneName=head, BaseName=spine_05, DeltaComponentType=QuaternionTwist, InputRange=(X=0,Y=50), TwistAxis="0,1,0")) -> (Output_Get:double)
   n2 = GetPartSkeletalMeshComponent()
   > Set movingUpLipRingAlpha(movingUpLipRingAlpha:double=Get Distance Between Two Sockets!(
     Component:Skeletal Mesh Component Object Reference=@n2
     SocketOrBoneNameA:Name=def_uppper_lip_PST
     SocketSpaceA:ERelativeTransformSpace Enum=RTS_World
     SocketOrBoneNameB:Name=def_lower_lip_PST
     SocketSpaceB:ERelativeTransformSpace Enum=RTS_World
     bRemapRange:bool=false
     InRangeMin:float=0.24
     InRangeMax:float=0.6
     OutRangeMin:float=0
     OutRangeMax:float=1
    ) -> (ReturnValue:float) {is=title K2Node_CallFunction}) -> (Output_Get:double)
   Event OnCharacterCustomizationLoaded()
   > Do Once()
   n3 = CheckForBackbling()
   > Set isWearBackbling(isWearBackbling:bool=@n3) -> (Output_Get:bool)
   > Set isWearCape(isWearCape:bool=@n3.bIsBackblingLong) -> (Output_Get:bool)
   Event BlueprintInitializeAnimation()
   n4 = GetOwnerSkeletalMeshComponent()
   > Is Valid(InputObject=@n4)
   > Set SkelMesh(SkelMesh=@n4)
   Get Distance Between Two Sockets(
    Component=@n2
    SocketOrBoneNameA=FACIAL_R_LipUpper
    SocketSpaceA=RTS_World
    SocketOrBoneNameB=FACIAL_C_LipLower
    SocketSpaceB=RTS_World
    bRemapRange=false
    InRangeMin=1.95
    InRangeMax=2.3
    OutRangeMin=0
    OutRangeMax=1
   )
  ]}
]
variables [
 head_rot_bk {type=number}
 head_rot_r {type=number}
 isWearBackbling {type=boolean}
 isWearCape {type=boolean}
 movingUpLipRingAlpha {type=number}
 SkelMesh {type=SkeletalMeshComponent}
]
